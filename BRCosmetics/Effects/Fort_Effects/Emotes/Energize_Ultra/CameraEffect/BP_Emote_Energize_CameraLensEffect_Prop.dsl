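; Emote prop for Energize Ultra that plays a camera effect when its attached parent pawn is locally viewed, skipped in frontend display.
; After a short delay, a timeline animates post-process settings, including a chromatic track, and a camera lens effect spawns.
; Outside mobile and PC performance mode, it also adds a dynamic blur material as a post-process blendable, with a second timeline driving offset and center.
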(event EventBeginPlay
  (bind _dynamic_material_instance_setting (Variables|Default|GetDynamic%20material%20instance%20setting))
  (bind _postprocess (Variables|Default|GetPostProcess))
  (if (not (Reactive|Frontend|IsCosmeticObjectDisplayedinFrontendContext 0))
    (Utilities|FlowControl|Delay 0.1)
    (bind _asfort_pawn (Utilities|Casting|CastToFortPawn (Actor|GetAttachParentActor self)))
    (if (Pawn|IsLocallyViewed _asfort_pawn)
      (Utilities|FlowControl|Sequence
        (:then_0
          ;[1 1 1] PostProcess
          (bind (_direction _chromatic _newtrack_1) (Timeline|Timeline PlayFromStart)
            (:Update
              (bind _structout (Utilities|Struct|SetmembersinPostProcessSettings (Variables|/Script/Engine.PostProcessComponent|GetSettings _postprocess) (Math|Vector4|MakeVector4 _newtrack_1 _newtrack_1 _newtrack_1 _newtrack_1) _chromatic))
              (Variables|/Script/Engine.PostProcessComponent|SetSettings _structout _postprocess))))
        (:then_1
          (bind _returnvalue (Settings|GetGameUserSettings))
          (switch int (Settings|GetAntiAliasingQuality _returnvalue)
            (:3
              _)
            (:4
              _))
          (switch int (Settings|GetVisualEffectQuality _returnvalue)
            (:2
              _)
            (:3
              _)
            (:4
              _))
          (switch int (Settings|GetPostProcessingQuality _returnvalue)
            (:2
              _)
            (:3
              _)
            (:4
              _))
          (if (Fortnite|IsMobilePlatform)
            _
            (elif (Class|PlatformHelpersLibrary|IsPcPerfMode)
              _
              (else
                ;[1 1 1] PostProcess material
                (bind _returnvalue_2 (Rendering|Material|CreateDynamicMaterialInstance "/Game/Effects/Fort_Effects/Emotes/Energize_Ultra/CameraEffect/MI_PP_Camera_Blur.MI_PP_Camera_Blur" "None" "None" :!class /Script/Engine.KismetMaterialLibrary))
                (bind _output_get (Variables|Default|SetDynamic%20material%20instance%20setting _returnvalue_2))
                (Variables|/Script/Engine.PostProcessComponent|SetSettings (Utilities|Struct|MakePostProcessSettings (Utilities|Struct|MakeWeightedBlendables (Utilities|Array|MakeArray (Utilities|Struct|MakeWeightedBlendable 1.0 _output_get)))) (Variables|Default|GetPostProcess))
                (bind (_direction_2 _blur _blur_center) (Timeline|Timeline_0 PlayFromStart)
                  (:Update
                    (Rendering|Material|SetScalarParameterValue _dynamic_material_instance_setting "OffsetAmount" _blur :!class /Script/Engine.MaterialInstanceDynamic)
                    (Rendering|Material|SetScalarParameterValue _dynamic_material_instance_setting "BlurCenter" _blur_center :!class /Script/Engine.MaterialInstanceDynamic)))))))
        (:then_2
          (bind _returnvalue_3 (Feedback|SpawnGenericCameraLensEffect _asfort_pawn "/BRCosmetics/Effects/Fort_Effects/Emotes/Energize_Ultra/CameraEffect/BP_Emote_Energize_CameraEffect.BP_Emote_Energize_CameraEffect_C"))
          (Variables|Default|SetCameraLensFX _returnvalue_3))))))

(event EventDestroyed
  (Actor|DestroyActor))

(fn ConstructionScript)
